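# MetaSound Source preset that plays the Troop Factor diegetic music loop from a TV as positional stereo audio.
# It inherits its graph from MSS_MusicEncounter_DiageticLoop and only supplies the looping wave asset, tempo, loop length and a zero fade-in time.
# A dedicated TV attenuation asset handles spatial falloff, and the NPC music sound class handles routing.
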MetaSoundSource /WildEstate_TroopFactor/Audio/IP_MSS_TroopFactor_Music_TV
preset_parent=/MusicEncounter/MetaSounds/MSS_MusicEncounter_DiageticLoop
inputs [
 FadeInTime {type=Time}
 Music.LoopLength {type=Int32, access=Constructor}
 Music.Tempo {type=Float}
 Music.WaveAsset {type=WaveAsset}
 UE.Source.OnPlay {type=Trigger}
]
interfaces [UE.OutputFormat.Stereo:1.0, UE.Source:1.1]
outputs [
 "UE.OutputFormat.Stereo.Audio:0" {type=Audio}
 "UE.OutputFormat.Stereo.Audio:1" {type=Audio}
]
pages [Default {input_defaults {
   FadeInTime=0
   Music.LoopLength=52
   Music.Tempo=165
   Music.WaveAsset=/WildEstate_TroopFactor/Audio/SoundWaves/MX_S38_IP_TroopFactor_Diegetic_LP
   UE.Source.OnPlay=null
  }}]
properties {
 assetImportData=&self:AssetImportData
 attenuationSettings=/WildEstate_TroopFactor/Audio/Settings/ATT_TroopFactor_TV
 bLooping=true
 duration=10000
 maxDistance=2000
 numChannels=2
 outputFormat=Stereo
 sampleRate=48000
 soundClassObject=&Fort_Music_NPC
 assetUserData [&self:AudioAssetUserData_0]
}
